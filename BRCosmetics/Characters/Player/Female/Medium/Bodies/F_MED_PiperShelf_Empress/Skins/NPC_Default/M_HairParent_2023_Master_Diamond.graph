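# Master character hair material for the F_MED_PiperShelf_Empress NPC skin, adding a diamond transition effect.
# Character PreFX output is blended with Fortnite 2023 hair by hair mask, then optionally passed through flyaway hair, both behind static switches.
# Base color lerps toward an override before MF_NewDiamond, whose emissive is scaled by a lerped boost before character PostFX.
# Diamond_Alpha drives the named Alpha and TechArtMask_Blend reroutes for the transition.
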Material /BRCosmetics/Characters/Player/Female/Medium/Bodies/F_MED_PiperShelf_Empress/Skins/NPC_Default/M_HairParent_2023_Master_Diamond
graph {
 properties {
  bUseMaterialAttributes=true
  bUsedWithInstancedStaticMeshes=true
  bUsedWithMeshDeformer=true
  bUsedWithSkeletalMesh=true
  pixelDepthOffsetMode=PDOM_Legacy
  shadingModel=MSM_FromMaterialExpression
  usedShadingModels=Default Lit | Subsurface | Clear Coat | Subsurface Profile | Cloth
 }
 nodes [
  n1 = MaterialFunctionCall() {materialFunction=&MF_CharacterParent_2023_PreFX}
  n2 = MaterialFunctionCall!(In=@n1.Result, Mask Input=@n1.Result) {materialFunction=/Game/Global/Materials/Character_Parent_2023/Hair_Parent_2023/MaterialFunctions/MF_FN_Hair_2023}
  n3 = BlendMaterialAttributes(
   A=@n1.Result
   B=StaticSwitchParameter(True=@n2.Result, False=@n1.Result) {defaultValue=true, group=Skin FX, parameterName=UseFNHair2_0}
   Alpha=@n2.HairMask
  )
  n4 = GetMaterialAttributes(None=StaticSwitchParameter(
    True=MaterialFunctionCall!(MaterialInput=@n3) {materialFunction=/Game/Global/Materials/Character_Parent_2023/Hair_Parent_2023/MaterialFunctions/MF_FlyAwayHair_2023}.Result
    False=@n3
   ) {group=Skin FX, parameterName=IsFlyAwayHair?}) {attributeGetTypes [#1]}
  n5 = LinearInterpolate(A=@n4.Base Color, B=ScalarParameter() {defaultValue=0.35, group=DiamondTransit, parameterName=BaseColorOverride}, Alpha=NamedRerouteUsage()) {constB=0.35}
  n6 = GetMaterialAttributes(None=MaterialFunctionCall!(
    Material=SetMaterialAttributes!(MaterialAttributes=@n4.MaterialAttributes, Base Color=@n5) {attributeSetTypes [#1]}
    TechArtMask=NamedRerouteUsage()
   ) {materialFunction=/BRCosmetics/Characters/Player/Female/Medium/Bodies/F_MED_PiperShelf_Empress/Materials/MF_NewDiamond}.Material) {attributeGetTypes [#2]}
  n7 = LinearInterpolate(B=ScalarParameter() {defaultValue=2, group=DiamondTransit, parameterName=EmissiveOverride}, Alpha=NamedRerouteUsage()) {constA=1}
  MaterialFunctionCall(
   Material=SetMaterialAttributes!(MaterialAttributes=@n6.MaterialAttributes, Emissive Color=Multiply(A=@n6.Emissive Color, B=@n7)) {attributeSetTypes [#2]}
   M=@n1.M
   Pre Skinned Local Position=@n1.M
   Distance Blend=@n1.M
   Fresnel=@n1.M
  ) {materialFunction=&MF_CharacterParent_2023_PostFX}
  n8 = ScalarParameter() {group=DiamondTransit, parameterName=Diamond_Alpha}
  NamedRerouteDeclaration(None=LinearInterpolate(A=Constant(), Alpha=NamedRerouteDeclaration(None=@n8) {name=Alpha, nodeColor=<rgba 1 0 0.9375 1>})) {name=TechArtMask_Blend, nodeColor=<rgba 0.2188 1 0 1>}
  Reroute(None=@n1.Tech Art Mask)
 ]
}
